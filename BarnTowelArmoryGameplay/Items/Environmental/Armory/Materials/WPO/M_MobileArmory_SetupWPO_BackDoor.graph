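# Nanite-enabled material for the Mobile Armory back door that animates the door through World Position Offset.
# Time since BP_GetTime, adjusted by delay and speed, is saturated and squared into an ease-in alpha; the huge default time holds the start pose.
# That alpha drives rotation about local X at BackDoor_Pivot plus a local Y slide from DoorOffset_Depth.
# The normal map rotates with the door, and SMR green feeds metallic, specular and roughness.
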Material /BarnTowelArmoryGameplay/Items/Environmental/Armory/Materials/WPO/M_MobileArmory_SetupWPO_BackDoor
graph {
 properties {
  bAutomaticallySetUsageInEditor=false
  bUseMaterialAttributes=true
  bUsedWithClothing=true
  bUsedWithNanite=true
  maxWorldPositionOffsetDisplacement=5000
  usedShadingModels=Default Lit
 }
 nodes [
  n1 = Time()
  Frac(None=Multiply(A=@n1) {constB=0.5})
  n2 = TextureSampleParameter2D() {group=01_PBR, parameterName=DiffuseColor, texture=/BarnTowelArmoryGameplay/Items/Environmental/Armory/Textures/T_MobileArmoryDetails_E}
  n3 = TextureSampleParameter2D() {group=01_PBR, parameterName=SMR, texture=&T_MobileArmoryDetails_E}
  n4 = TextureSampleParameter2D() {group=01_PBR, parameterName=EmissiveTexture, texture=&T_MobileArmoryDetails_E}
  n5 = Subtract(A=@n1, B=ScalarParameter() {group=WPO_Door, parameterName=DoorAnimationDelay, sortPriority=40})
  n6 = ScalarParameter() {defaultValue=99999997952, group=WPO_Door, parameterName=BP_GetTime}
  n7 = Saturate(None=Multiply(A=Subtract(A=@n5, B=@n6), B=ScalarParameter() {group=WPO_Door, parameterName=DoorAnimationSpeed, sortPriority=40}))
  n8 = Multiply(A=@n7, B=@n7)
  n9 = TextureSampleParameter2D() {
   group=01_PBR
   parameterName=NormalTexture
   samplerType=SAMPLERTYPE_Normal
   texture=/BarnTowelArmoryGameplay/Items/Environmental/Armory/Textures/T_MobileArmoryDetails_N
  }
  n10 = MaterialFunctionCall!(
   RotationAngle=LinearInterpolate(Alpha=@n8) {constA=-0.2222, constB=0}
   NormalizedRotationAxis=Transform(None=Constant3Vector() {constant=<rgba 1 0 0 0>}) {transformSourceType=TRANSFORMSOURCE_Local}
   PivotPoint=TransformPosition(None=VectorParameter() {group=WPO_Door, parameterName=BackDoor_Pivot, defaultValue=<rgba 0 -608 728 1>}.RGB) {transformType=TRANSFORMPOSSOURCE_World}
   Position
   Normal=Transform(None=@n9.RGB)
   TangentU
   TangentV
  ) {materialFunction=/Game/Characters/Player/Male/Medium/Bodies/M_MED_Inferno/FX/Mat/RotateVertexAboutAxis}
  n11 = Multiply(
   A=Constant3Vector() {constant=<rgba 0 1 0 0>}
   B=LinearInterpolate(A=ScalarParameter() {group=WPO_Door, parameterName=DoorOffset_Depth}, Alpha=@n8)
  )
  MakeMaterialAttributes!(
   BaseColor=@n2.RGB
   Metallic=@n3.G
   Specular=@n3.G
   Roughness=@n3.G
   Anisotropy
   EmissiveColor=@n4.RGB
   Opacity
   OpacityMask
   Normal=Transform(None=@n10.Normal) {transformSourceType=TRANSFORMSOURCE_World, transformType=TRANSFORM_Tangent}
   Tangent
   WorldPositionOffset=Add(A=Transform(None=@n11) {transformSourceType=TRANSFORMSOURCE_Local}, B=@n10.WPO)
   SubsurfaceColor
   ClearCoat
   ClearCoatRoughness
   AmbientOcclusion
   Refraction
   CustomizedUVs_0
   CustomizedUVs_1
   CustomizedUVs_2
   CustomizedUVs_3
   CustomizedUVs_4
   CustomizedUVs_5
   CustomizedUVs_6
   CustomizedUVs_7
   PixelDepthOffset
   ShadingModel
   Displacement
  )
 ]
}
