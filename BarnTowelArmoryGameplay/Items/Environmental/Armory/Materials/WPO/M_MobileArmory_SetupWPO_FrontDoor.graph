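# Material for the mobile armory front door that animates the door during setup through World Position Offset.
# Time minus a delay and the BP_GetTime start parameter is scaled by speed, clamped and squared to ease the door along a local-space offset.
# On completion, an If node swaps in a huge negative offset, effectively removing the door from view.
# PBR textures supply shading, with SMR green driving metallic, specular and roughness alike.
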Material /BarnTowelArmoryGameplay/Items/Environmental/Armory/Materials/WPO/M_MobileArmory_SetupWPO_FrontDoor
graph {
 properties {
  bAutomaticallySetUsageInEditor=false
  bUseMaterialAttributes=true
  bUsedWithClothing=true
  bUsedWithNanite=true
  maxWorldPositionOffsetDisplacement=5000
  usedShadingModels=Default Lit
 }
 nodes [
  n1 = TextureSampleParameter2D() {group=01_PBR, parameterName=DiffuseColor, texture=/BarnTowelArmoryGameplay/Items/Environmental/Armory/Textures/T_MobileArmoryDetails_E}
  n2 = TextureSampleParameter2D() {group=01_PBR, parameterName=SMR, texture=&T_MobileArmoryDetails_E}
  n3 = TextureSampleParameter2D() {group=01_PBR, parameterName=EmissiveTexture, texture=&T_MobileArmoryDetails_E}
  n4 = TextureSampleParameter2D() {
   group=01_PBR
   parameterName=NormalTexture
   samplerType=SAMPLERTYPE_Normal
   texture=/BarnTowelArmoryGameplay/Items/Environmental/Armory/Textures/T_MobileArmoryDetails_N
  }
  n5 = Subtract(A=Time(), B=ScalarParameter() {group=WPO_Door, parameterName=DoorAnimationDelay, sortPriority=40})
  n6 = ScalarParameter() {defaultValue=99999997952, group=WPO_Door, parameterName=BP_GetTime}
  n7 = Saturate(None=Multiply(A=Subtract(A=@n5, B=@n6), B=ScalarParameter() {group=WPO_Door, parameterName=DoorAnimationSpeed, sortPriority=40}))
  n8 = Multiply(A=@n7, B=@n7)
  n9 = Transform(None=Constant3Vector() {constant=<rgba 0 25 365 0>}) {transformSourceType=TRANSFORMSOURCE_Local}
  MakeMaterialAttributes!(
   BaseColor=@n1.RGB
   Metallic=@n2.G
   Specular=@n2.G
   Roughness=@n2.G
   Anisotropy
   EmissiveColor=@n3.RGB
   Opacity
   OpacityMask
   Normal=@n4.RGB
   Tangent
   WorldPositionOffset=If(A=@n8, "A > B"=Constant() {r=-999999986991104}, "A < B"=LinearInterpolate(B=@n9, Alpha=@n8)) {constB=1}
   SubsurfaceColor
   ClearCoat
   ClearCoatRoughness
   AmbientOcclusion
   Refraction
   CustomizedUVs_0
   CustomizedUVs_1
   CustomizedUVs_2
   CustomizedUVs_3
   CustomizedUVs_4
   CustomizedUVs_5
   CustomizedUVs_6
   CustomizedUVs_7
   PixelDepthOffset
   ShadingModel
   Displacement
  )
  ScalarParameter() {group=01_PBR, parameterName=Metallic, sortPriority=1}
  ScalarParameter() {defaultValue=0.5, group=01_PBR, parameterName=Specular, sortPriority=2}
  ScalarParameter() {defaultValue=0.5, group=01_PBR, parameterName=Roughness, sortPriority=2}
 ]
}
